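# Physics asset for the Clown Backpack cosmetic, giving its two balloon bases simple ragdoll-style collision and joint limits.
# It has one capsule body on the backpack root and one on each of the two dynamic balloon base bones.
# Each balloon body is constrained directly to the root with equal, moderate swing and twist limits, so the two are jointed independently rather than chained.
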PhysicsAsset /BRCosmetics/Accessories/FORT_Backpacks/F_MED_Clown_Backpack/Meshes/F_MED_Clown_Backpack_Physics
bodies {
 backpack_root [{shapeType=Capsule, length=13.1104, radius=12.2519, center=<-0.0795 -7.4258 -1.2366>, rotation=<rot -7.1912 -90 -42.0249>}]
 dyn_balloon_base_01_01 [{shapeType=Capsule, length=15.8278, radius=10.0984, center=<-1.9385 0.3266 -2.7966>, rotation=<rot 0 83.9867 -89.712>}]
 dyn_balloon_base_02 [{shapeType=Capsule, length=24.1311, radius=9.4204, center=<-0.5745 -3.1334 4.3852>, rotation=<rot 13.5314 -90 1.0453>}]
}
constraints [
 {bone1Name=dyn_balloon_base_01_01, bone2Name=backpack_root, swing1LimitDegrees=45, swing2LimitDegrees=45, twistLimitDegrees=45}
 {bone1Name=dyn_balloon_base_02, bone2Name=backpack_root, swing1LimitDegrees=45, swing2LimitDegrees=45, twistLimitDegrees=45}
]
