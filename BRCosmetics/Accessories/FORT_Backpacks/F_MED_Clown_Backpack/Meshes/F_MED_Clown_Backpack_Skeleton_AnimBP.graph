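# Animation Blueprint for the clown backpack cosmetic, animating its balloon bones with procedural secondary motion.
# The event graph derives pawn yaw and a skydiving force from Perlin noise on pawn height, damped under an open parachute.
# Two Anim Dynamics nodes take that force, then a world-yaw bone, Trail controller with planar limits and Look At finish the pose.
# Several Trail and transform nodes are left disconnected from the output.
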AnimBlueprint /BRCosmetics/Accessories/FORT_Backpacks/F_MED_Clown_Backpack/Meshes/F_MED_Clown_Backpack_Skeleton_AnimBP
target_skeleton=&/F_MED_Clown_Backpack_Skeleton
graphs [
 AnimGraph {nodes [
   n1 = "Transform (Modify) Bone - Bone: dyn_balloon_base_02"!(
    ComponentPose=Copy Bone!(
     ComponentPose:Component Space Pose Link Structure
     Alpha:float=1
     bCopyTranslation:bool=False
     bCopyRotation:bool=False
     bCopyScale:bool=False
    ) -> (Pose:Component Space Pose Link Structure) {is=title AnimGraphNode_CopyBone, alphaBoolBlend {blendOption=Linear}}
    Alpha=0
    Rotation="0,0,0"
   ) -> (Pose) {rotationMode=BMM_Replace, rotationSpace=BCS_WorldSpace, alphaBoolBlend {blendOption=Linear}, boneToModify {boneName=dyn_balloon_base_02}}
   n2 = Trail controller - Bone: dyn_balloon_base_02_01!(ComponentPose=@n1, Alpha=1, FakeVelocity:Vector="0,0,0") -> (Pose) {
    bReorientParentToChild=false
    bUsePlanarLimit=true
    relaxationSpeedScale=2
    baseJoint {boneName=dyn_balloon_base_02}
    planarLimits [
     {drivingBone {boneName=dyn_balloon_base_02}, planeTransform {location=<0 0 9>, rotation=<rot 0 0 0>, scale=<1 1 1>}}
     {drivingBone {boneName=backpack_root}, planeTransform {location=<20 -8 0>, rotation=<rot 0 5.0001 -89.9999>, scale=<1 1 1>}}
    ]
    rotationLimits [{limitMax=<180 180 180>}, {limitMax=<180 180 180>}]
    rotationOffsets [<0 0 0>, <0 0 0>]
    trailBone {boneName=dyn_balloon_base_02_01}
   }
   Component To Local(ComponentPose=@n2)
   n3 = Get falling_anim_dyn_up_force_vec3() -> (falling_anim_dyn_up_force_vec3:Vector)
   n4 = Anim Dynamics - Bone: dyn_balloon_base_01_02(ComponentPose=Local To Component(), Alpha=0.5, ExternalForce=vector * float(A=@n3, B=0.0275)) {
    angularSpringConstant=1000
    bAngularSpring=true
    bGravityOverrideInSimSpace=true
    bLinearSpring=true
    linearSpringConstant=1000
    alphaBoolBlend {blendOption=Linear}
    boundBone {boneName=dyn_balloon_base_01_02}
    componentAppliedLinearAccClamp=<20 20 20>
    componentLinearAccScale=<1 1 1>
    componentLinearVelScale=<1 1 1>
    physicsBodyDefinitions [{
      sphereCollisionRadius=0
      boundBone {boneName=dyn_balloon_base_01_02}
      boxExtents=<1 1 1>
      constraintSetup {angularLimitsMax=<0 0 0>, angularLimitsMin=<0 0 -64.7999>, angularTarget=<0 0 0>, linearAxesMax=<0.1 1 0>, linearAxesMin=<0 -1 0>}
      localJointOffset=<0 0 -3>
     }]
    relativeSpaceBone {boneName=dyn_balloon_base_01_A}
   }
   n5 = Anim Dynamics - Bone: dyn_balloon_base_01_03(ComponentPose=@n4, Alpha=0.5, ExternalForce=vector * float(A=@n3, B=-0.03)) {
    angularSpringConstant=1000
    bAngularSpring=true
    bGravityOverrideInSimSpace=true
    bLinearSpring=true
    linearSpringConstant=1000
    alphaBoolBlend {blendOption=Linear}
    boundBone {boneName=dyn_balloon_base_01_03}
    componentAppliedLinearAccClamp=<40 40 40>
    componentLinearAccScale=<1 1 1>
    componentLinearVelScale=<1 1 1>
    physicsBodyDefinitions [{
      sphereCollisionRadius=0
      boundBone {boneName=dyn_balloon_base_01_03}
      boxExtents=<1 1 1>
      constraintSetup {angularLimitsMax=<0 0 0>, angularLimitsMin=<0 0 0>, angularTarget=<0 0 0>, linearAxesMax=<0 1 0>, linearAxesMin=<-0.1 -1 0>}
      localJointOffset=<0 0 -3>
     }]
    relativeSpaceBone {boneName=dyn_balloon_base_01_B}
   }
   n6 = "Transform (Modify) Bone - Bone: dyn_balloon_base_02"(ComponentPose=@n5, Alpha=1, Rotation=Get pawnYaw() -> (pawnYaw:Rotator)) {rotationMode=BMM_Replace, rotationSpace=BCS_WorldSpace, alphaBoolBlend {blendOption=Linear}, boneToModify {boneName=dyn_balloon_base_02}}
   n7 = Trail controller - Bone: dyn_balloon_base_02_01(ComponentPose=@n6, Alpha=1, FakeVelocity=vector * float(A=@n3, B=0.01)) {
    bUsePlanarLimit=true
    chainBoneAxis=Z
    relaxationSpeedScale=2
    baseJoint {boneName=dyn_balloon_base_02}
    planarLimits [
     {drivingBone {boneName=dyn_balloon_base_02}, planeTransform {location=<0 0 14>, rotation=<rot -33 0 -22>, scale=<1 1 1>}}
     {drivingBone {boneName=backpack_root}, planeTransform {location=<20 -8 0>, rotation=<rot 0 5.0001 -89.9999>, scale=<1 1 1>}}
    ]
    rotationLimits [{limitMax=<180 180 180>}, {limitMax=<180 180 180>}]
    rotationOffsets [<0 0 0>, <0 0 0>]
    trailBone {boneName=dyn_balloon_base_02_01}
   }
   n8 = Look At - Bone: dyn_balloon_base_02_03_Aim(ComponentPose=@n7, Alpha=1) {
    alphaBoolBlend {blendOption=Linear}
    boneToModify {boneName=dyn_balloon_base_02_03_Aim}
    lookAtLocation {x=0}
    lookAtTarget {boneReference {boneName=dyn_balloon_base_02}}
    lookAt_Axis {axis {y=0, z=-1}}
    lookUp_Axis {axis=<0 1>}
   }
   Output Pose(Result=Component To Local(ComponentPose=@n8)) {layerGroup=None}
   n9 = Trail controller - Bone: dyn_balloon_base_01_02(Alpha=0.25) {
    relaxationSpeedScale=2
    stretchLimit=0.1
    baseJoint {boneName=dyn_balloon_base_01_A}
    planarLimits [{drivingBone {boneName=dyn_balloon_base_01_A}, planeTransform {location=<0 0 0>, rotation=<rot -45 0 0>, scale=<1 1 1>}}]
    rotationLimits [{limitMax=<180 180 180>}, {limitMax=<180 180 180>}]
    rotationOffsets [<0 0 0>, <0 0 0>]
    trailBone {boneName=dyn_balloon_base_01_02}
   }
   Trail controller - Bone: dyn_balloon_base_01_03(ComponentPose=@n9, Alpha=0.25) {
    relaxationSpeedScale=2
    stretchLimit=0.1
    baseJoint {boneName=dyn_balloon_base_01_B}
    rotationLimits [{limitMax=<180 180 180>}, {limitMax=<180 180 180>}]
    rotationOffsets [<0 0 0>, <0 0 0>]
    trailBone {boneName=dyn_balloon_base_01_03}
   }
   n10 = NormalizeToRange!(Value:double=BreakVector(InVec=@n3).Y, RangeMin:double=0, RangeMax:double=2500) -> (ReturnValue:double)
   float + float(A=float / float(A=1, B=float + float(A=@n10, B=1.1)) {is=title K2Node_CallFunction}, B=1.1) {is=title K2Node_CommutativeAssociativeBinaryOperator}
  ]}
 EventGraph {nodes [
   Event BlueprintUpdateAnimation()
   n1 = Get PawnLastLocation() -> (PawnLastLocation:Vector)
   n2 = SelectFloat(
    A=Make Perlin Noise and Remap!(Value:float=BreakVector(InVec=@n1).Z, RangeOutMin:float=-25000, RangeOutMax:float=25000) -> (ReturnValue:float)
    B=0
    bPickA=Get bIsSkyDiving()
   )
   n3 = SelectFloat(A=float * float(A=@n2, B=0.1), B=@n2, bPickA=Get bIsParachuteOpened())
   > Set falling_anim_dyn_up_force_vec3(falling_anim_dyn_up_force_vec3:Vector=vector * float(A=MakeVector(Y=@n3), B=0.75)) -> (Output_Get:Vector)
   n4 = float * float(A=BreakRotator(InRot=Get PawnLastRotation() -> (PawnLastRotation:Rotator)).Yaw, B=1)
   > Set pawnYaw(pawnYaw:Rotator=MakeRotator(Yaw=float + float(A=@n4, B=-90))) -> (Output_Get:Rotator)
  ]}
]
variables [
 falling_anim_dyn_up_force_vec3 {type=Vector}
 pawnYaw {type=Rotator}
]
